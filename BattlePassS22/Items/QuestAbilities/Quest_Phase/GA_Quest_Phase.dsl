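; Gameplay ability for a Season 22 Battle Pass quest that credits the owning player for phasing through phaseable builds, directly or while riding.
; When triggered, it calls the parent and switches on the event's gameplay tag.
; A player phase event sends quest progress to the owner's quest manager immediately.
; A riding event waits for a phased event on the pawn's valid attach parent before sending the same progress.
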(event Ability|EventActivateAbilityFromEvent
  (params (EventData :type Gameplay%20Event%20Data%20Structure%20%28by%20ref%29))
  (bind _owningplayerpawn (Variables|GAVars|Data|GetOwningPlayerPawn))
  (bind _returnvalue (Actor|GetAttachParentActor _owningplayerpawn))
  (|CallParentFunction EventData)
  (switch Utilities|FlowControl|Switch|SwitchonGameplayTag (Utilities|Struct|BreakGameplayEventData EventData)
    (:Default
      _)
    (:Event.PhaseableBuildGameplay.PhaseableBuild.Player.Phase
      (Quest|SendComplexCustomStatEvent (Variables|GAVars|Data|GetOwningPlayerQuestManager) 0 (Variables|Default|GetT_QuestProgress) "(GameplayTags=)" 1))
    (:Riding.IsRidingActor
      (Utilities|IsValid _returnvalue
        (:Is%20Valid
          (Ability|Tasks|WaitGameplayEvent (Variables|Default|GetT_EventPhased) _returnvalue
            (:EventReceived
              (Quest|SendComplexCustomStatEvent (Variables|GAVars|Data|GetOwningPlayerQuestManager) 0 (Variables|Default|GetT_QuestProgress) "(GameplayTags=)" 1))))
        (:Is%20Not%20Valid
          _))))
  (Ability|EndAbility))
